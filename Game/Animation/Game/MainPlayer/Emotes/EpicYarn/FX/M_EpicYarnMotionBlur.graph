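# M_EpicYarnMotionBlur is a translucent, two-sided material for Niagara mesh particles, apparently drawing motion-blur streaks for the EpicYarn emote.
# A dynamic SplineWidth parameter drives SplineThicken to widen the mesh.
# Panning blurred noise is multiplied by a V gradient and two sphere masks, then scaled, smoothstepped and tinted by particle color and alpha.
# SmoothStep takes SmoothStepMin for both Min and Max, leaving SmoothStepMax unused.
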Material /Game/Animation/Game/MainPlayer/Emotes/EpicYarn/FX/M_EpicYarnMotionBlur
graph {
 properties {
  bUsedWithNiagaraMeshParticles=true
  blendMode=BLEND_Translucent
  pixelDepthOffsetMode=PDOM_Legacy
  twoSided=true
  usedShadingModels=Default Lit
 }
 nodes [
  n1 = DynamicParameter() {defaultValue=<rgba 5 1 1 1>, paramNames [SplineWidth, SmoothStepMin, SmoothStepMax, Param4]}
  MaterialFunctionCall(WidthBase=@n1.SplineWidth, WidthTip=@n1.SplineWidth) {materialFunction=&SplineThicken}
  n2 = DynamicParameter() {defaultValue=<rgba 5 1 1 1>, paramNames [SplineWidth, SmoothStepMin, SmoothStepMax, Param4]}
  n3 = TextureCoordinate() {uTiling=1.5, vTiling=0.25}
  n4 = ComponentMask(None=TextureCoordinate()) {g=true}
  n5 = OneMinus(None=@n4)
  n6 = SphereMask(
   A=Add(A=Multiply(A=@n5) {constB=0.225}, B=TextureCoordinate() {uTiling=0})
   B=Constant2Vector() {g=0.5}
   Radius=Constant() {r=0.5}
   Hardness=Constant()
  )
  n7 = TextureCoordinate() {vTiling=0}
  n8 = Constant2Vector() {r=0.5}
  n9 = Constant() {r=0.5}
  n10 = Constant() {r=0.25}
  n11 = Multiply(
   A=TextureSample(UVs=Panner(Coordinate=@n3) {speedX=0.25, speedY=-1.5}) {samplerType=SAMPLERTYPE_Grayscale, texture=&LowResBlurredNoise}.R
   B=Multiply(A=Multiply(A=@n4, B=@n6), B=SphereMask(A=@n7, B=@n8, Radius=@n9, Hardness=@n10))
  )
  n12 = Saturate(None=SmoothStep(Min=@n2.SmoothStepMin, Max=@n2.SmoothStepMin, Value=ConstantBiasScale(None=@n11) {bias=0, scale=5}))
  n13 = ParticleColor()
  Multiply(A=@n12, B=@n13.RGB)
  Multiply(A=@n13.A, B=@n12)
 ]
}
